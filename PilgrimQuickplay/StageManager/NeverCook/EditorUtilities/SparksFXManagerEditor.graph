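# Editor utility widget for creating BP_MojoFxManager actors in the level and assigning stage show actors to them.
# A searchable dropdown, refilled by actor query, sets the current manager shown in a details view.
# Tick polls the editor selection for StageShowFXInterface implementers, updating button states, counts and tooltips.
# Add buttons cast those actors to StageShowActorBase, append them uniquely to the manager's StageShowActors and call TransactObject.
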EditorUtilityWidgetBlueprint /PilgrimQuickplay/StageManager/NeverCook/EditorUtilities/SparksFXManagerEditor
graphs [
 EventGraph {nodes [
   "On Clicked (CreateFxButton)"() {is=title K2Node_ComponentBoundEvent}
   > n1 = SpawnActor BP Mojo Fx Manager!(
    execute:Exec
    WorldContextObject:Object Reference
    Class:Actor Class Reference=/SparksStageShow/Systems/StageManager/FXManager/BP_MojoFxManager.BP_MojoFxManager_C
    SpawnTransform:Transform=MakeTransform(Rotation="0, 0, 0")
    CollisionHandlingOverride:ESpawnActorCollisionHandlingMethod Enum=Undefined
    TransformScaleMethod:ESpawnActorScaleMethod Enum=MultiplyWithRoot
    Owner:Actor Object Reference
    Instigator:Pawn Object Reference
   ) -> (then:Exec, ReturnValue:BP Mojo Fx Manager Object Reference) {is=title K2Node_SpawnActorFromClass}
   > n2 = Set CurrentFxManager(CurrentFxManager:BP Mojo Fx Manager Object Reference=@n1) -> (Output_Get:BP Mojo Fx Manager Object Reference)
   n3 = "To String (Text)"(InText="GetText (Text Box)"!(self:Text Box Object Reference=Get NewFxManagerNameTextBox() -> (NewFxManagerNameTextBox:Editor Utility Editable Text Box Object Reference)) -> (ReturnValue:Text))
   > SetActorLabel(self=@n2, NewActorLabel=@n3, bMarkDirty=true)
   > RefreshFxManagerList()
   > SetSelectedOption!(
    execute
    self:"Searchable Combo Box (String) Object Reference"=Get FxManagersDropDown() -> (FxManagersDropDown:"Searchable Combo Box (String) Object Reference")
    Option:String=@n3
   ) -> (then)
   > UpdateFxManagerVisibility!(execute, self, ShowCreate:bool=false) -> (then)
   n4 = "On Clicked (RefreshFxManagersListButton)"() {is=title K2Node_ComponentBoundEvent}
   n5 = ClearOptions(execute=[@n4.then, @n23.then], self:"Searchable Combo Box (String) Object Reference"=Get FxManagersDropDown())
   n6 = GetAllActorsOfClass!(
    execute=[@n5.then, @n24.then]
    WorldContextObject:Object Reference
    ActorClass:Actor Class Reference=&BP_MojoFxManager_C
   ) -> (then, OutActors:Array of BP Mojo Fx Manager Object References)
   > n7 = Set FxManagers(FxManagers:Array of BP Mojo Fx Manager Object References=@n6) -> (Output_Get:Array of BP Mojo Fx Manager Object References)
   > n8 = For Each Loop(Array:Array of BP Mojo Fx Manager Object References=@n7) -> (Array Element:BP Mojo Fx Manager Object Reference)
   n9 = GetDisplayName(Object="Get (a copy)"(Array:Array of BP Mojo Fx Manager Object References=@n7) -> (Output:BP Mojo Fx Manager Object Reference))
   AddOption!(execute=@n8.Completed, self:"Searchable Combo Box (String) Object Reference"=Get FxManagersDropDown(), Option:String=@n9) -> (then)
   > SetToolTipText!(execute, self:Widget Object Reference=Get FxManagersDropDown(), InToolTipText:"Text (by ref)"="To Text (String)"(InString=@n9)) -> (then)
   n10 = Get FxManagersDropDown()
   > n11 = Set CurrentFxManager(CurrentFxManager="Get (a copy)"(
     Array=Get FxManagers() -> (FxManagers:Array of BP Mojo Fx Manager Object References)
     Dimension 1=FindOptionIndex!(self:"Searchable Combo Box (String) Object Reference"=@n10, Option:String=@n9) -> (ReturnValue:int=0)
    ))
   > SetObject(self=Get FxManagerDetails() -> (FxManagerDetails:Details View Object Reference), NewObject=@n11)
   AddOption(execute=@n8.LoopBody, self=Get FxManagersDropDown(), Option=GetDisplayName(Object=@n8))
   n12 = "On Selection Changed (FxManagersDropDown)"!() -> (OutputDelegate, then, SelectedItem:String, SelectionType:ESelectInfo Enum) {is=title K2Node_ComponentBoundEvent}
   n13 = Get FxManagersDropDown()
   > Set CurrentFxManager(CurrentFxManager="Get (a copy)"(Array=Get FxManagers(), Dimension 1=FindOptionIndex(self=@n13, Option=@n12.SelectedItem) -> (ReturnValue=0)))
   > n14 = SetToolTipText(self=Get FxManagersDropDown(), InToolTipText="To Text (String)"(InString=@n12.SelectedItem))
   SetObject(
    execute=[@n14.then, @n28.then]
    self=Get FxManagerDetails()
    NewObject=Get CurrentFxManager() -> (CurrentFxManager:BP Mojo Fx Manager Object Reference)
   )
   "On Clicked (AddSelectedActorButton)"() {is=title K2Node_ComponentBoundEvent}
   > n15 = Get!(execute:Exec) -> (then:Exec, else:Exec, CurrentFxManager:BP Mojo Fx Manager Object Reference)
   > n16 = For Each Loop(Array:Array of Actor Object References=Get SelectedActors() -> (SelectedActors:Array of Actor Object References)) -> (Array Element:Actor Object Reference)
   > n17 = Cast To StageShowActorBase!(execute, Object=@n16) -> (then, CastFailed, AsStage Show Actor Base:Stage Show Actor Base Object Reference) {is=title K2Node_DynamicCast}
   > Add Unique(
    TargetArray:Array of Stage Show Actor Base Object References=Get StageShowActors!(self:Stage Show Actors Object Reference=@n15) -> (StageShowActors:Array of Stage Show Actor Base Object References)
    NewItem:"Stage Show Actor Base Object Reference (by ref)"=@n17
   )
   > TransactObject(Object=@n15)
   "On Clicked (CreateNewFxManagerButton)"() {is=title K2Node_ComponentBoundEvent}
   > UpdateFxManagerVisibility(ShowCreate=true)
   "On Clicked (CancelFxManButton)"() {is=title K2Node_ComponentBoundEvent}
   > UpdateFxManagerVisibility(ShowCreate=false)
   "On Clicked (CreateAndAddButton)"() {is=title K2Node_ComponentBoundEvent}
   > n18 = SpawnActor BP Mojo Fx Manager(
    Class=&BP_MojoFxManager_C
    SpawnTransform=MakeTransform(Rotation="0, 0, 0")
    CollisionHandlingOverride=Undefined
    TransformScaleMethod=MultiplyWithRoot
   )
   > n19 = Set CurrentFxManager(CurrentFxManager=@n18)
   > n20 = For Each Loop(Array=Get SelectedActors())
   n21 = "To String (Text)"(InText="GetText (Text Box)"(self=Get NewFxManagerNameTextBox()))
   SetActorLabel(execute=@n20.Completed, self=@n19, NewActorLabel=@n21, bMarkDirty=true)
   > RefreshFxManagerList()
   > SetSelectedOption(self=Get FxManagersDropDown(), Option=@n21)
   > UpdateFxManagerVisibility(ShowCreate=false)
   n22 = Cast To StageShowActorBase(execute=@n20.LoopBody, Object=@n20)
   > Add Unique(TargetArray=Get StageShowActors(self=@n19), NewItem=@n22)
   > TransactObject(Object=@n19)
   n23 = RefreshFxManagerList() {is=title K2Node_CustomEvent}
   Event Construct()
   > n24 = UpdateAddButtonStatus()
   Event Tick()
   > Clear(TargetArray:Array of Actor Object References=Get SelectedActors())
   > n25 = GetSelectedActors()
   > n26 = Sequence()
   > n27 = For Each Loop(Array=@n25)
   > Branch(Condition=Does Object Implement Interface!(
     TestObject:Object Reference=@n27
     Interface:Interface Interface=/SparksStageShow/Systems/StageManager/FXManager/StageShowFXInterface.StageShowFXInterface_C
    ) -> (ReturnValue:bool))
   > Add Unique(TargetArray:Array of Actor Object References=Get SelectedActors(), NewItem:"Actor Object Reference (by ref)"=@n27)
   UpdateAddButtonStatus(execute=@n26.then_1)
   n28 = Set ToolTipText!(execute, ToolTipText:Text, self:Widget Object Reference) -> (then, Output_Get:Text)
   Set ToolTipText()
  ]}
 UpdateAddButtonStatus {nodes [
   UpdateAddButtonStatus() {is=title K2Node_FunctionEntry}
   n1 = Length(TargetArray=Get SelectedActors())
   > SetIsEnabled(
    self=Get AddSelectedActorButton() -> (AddSelectedActorButton:Editor Utility Button Object Reference)
    bInIsEnabled="integer > integer"(A=@n1)
   )
   > "SetText (Text)"(
    self=Get AddSelectedActorText() -> (AddSelectedActorText:Text Object Reference)
    InText=Format Text!(Format="INVTEXT(\"({0}) Add Selected Actor to Manager\")", 0:int=@n1) -> (Result)
   )
   > "SetText (Text)"(self=Get CrateAndAddText() -> (CrateAndAddText:Text Object Reference), InText=Format Text(Format="INVTEXT(\"Create and Add ({0})\")", 0=@n1))
   n2 = TextIsEmpty(InText="GetText (Text Box)"(self=Get NewFxManagerNameTextBox()))
   n3 = Length(TargetArray=Get SelectedActors())
   > SetIsEnabled(
    self=Get CreateAndAddButton() -> (CreateAndAddButton:Editor Utility Button Object Reference)
    bInIsEnabled=AND Boolean(A=NOT Boolean(A=@n2), B="integer > integer"(A=@n3))
   )
   n4 = "GetText (Text Box)"(self=Get NewFxManagerNameTextBox())
   > SetIsEnabled(self=Get CreateFxButton() -> (CreateFxButton:Editor Utility Button Object Reference), bInIsEnabled=NOT Boolean(A=TextIsEmpty(InText=@n4)))
   > n5 = Sequence()
   n6 = Get SelectedActors()
   n7 = Select(
    Option 0:"Text (by ref)"=Format Text!(Format="INVTEXT(\"Please select a valid actor only actors that \\r\\nimplements Stage Show FX Interface are valid.\")") -> (Result)
    Option 1:"Text (by ref)"=Format Text(Format="INVTEXT(\"Add valid selected actor to the current selected fx manager from list\")")
    Index="integer > integer"(A=Length(TargetArray=@n6))
   ) -> (ReturnValue:"Text (by ref)")
   n8 = Get AddSelectedActorButton()
   > Branch(Condition="Not Equal Exactly (Text)"!(A:"Text (by ref)"=Get ToolTipText!(self:Widget Object Reference=@n8) -> (ToolTipText:Text), B:"Text (by ref)"=@n7) -> (ReturnValue:bool))
   > SetToolTipText(self=Get AddSelectedActorButton(), InToolTipText=@n7)
   n9 = Get SelectedActors()
   n10 = Select(
    Option 0=Format Text(Format="INVTEXT(\"Please select a valid actor only actors that \\r\\nimplements Stage Show FX Interface are valid.\")")
    Option 1=Format Text(Format="INVTEXT(\"Create a new FXManager and add a valid selected actor to it. \")")
    Index="integer > integer"(A=Length(TargetArray=@n9))
   )
   n11 = Get CreateAndAddButton()
   Branch(execute=@n5.then_1, Condition="Not Equal Exactly (Text)"(A=Get ToolTipText(self=@n11), B=@n10))
   > SetToolTipText(self=Get CreateAndAddButton(), InToolTipText=@n10)
  ]}
 UpdateFxManagerVisibility {nodes [
   n1 = UpdateFxManagerVisibility!() -> (then, ShowCreate:bool=~) {is=title K2Node_FunctionEntry}
   > n2 = Branch(Condition=@n1)
   > SetVisibility(
    self=[
     Get SelectFXManagerLine() -> (SelectFXManagerLine:Horizontal Box Object Reference)
     Get FxManagerDetailsScrollBox() -> (FxManagerDetailsScrollBox:Editor Utility Scroll Box Object Reference)
    ]
    InVisibility=Collapsed
   )
   > SetVisibility(self=Get CreateFxManagerLine() -> (CreateFxManagerLine:Horizontal Box Object Reference), InVisibility=SelfHitTestInvisible)
   SetVisibility(execute=@n2.else, self=[Get SelectFXManagerLine(), Get FxManagerDetailsScrollBox()], InVisibility=SelfHitTestInvisible)
   > SetVisibility(self=Get CreateFxManagerLine(), InVisibility=Collapsed)
  ]}
]
variables [
 CurrentFxManager {type=BP_MojoFxManager_C, default=null}
 FxManagers {type="array<BP_MojoFxManager_C>", default []}
 SelectedActors {type="array<Actor>"}
]
widgets {
 info {rootWidgetClass=&Overlay, widgetCount=30}
 widgets [
  Overlay_0 {widgetClassPath=&Overlay, widget=Overlay_0}
  VerticalBox_0 {
   widgetClassPath=&VerticalBox
   parent=Overlay_0
   widget=VerticalBox_0
   slot {class=OverlaySlot, properties {horizontalAlignment=HAlign_Fill, padding=<ltrb 25 50 25 0>}}
  }
  ObjectsSelectedText {
   widgetClassPath=&TextBlock
   bIsVariable=true
   parent=VerticalBox_0
   widget=ObjectsSelectedText
   properties {text=Text Block, visibility=Collapsed}
   slot {class=VerticalBoxSlot, properties {horizontalAlignment=HAlign_Left, verticalAlignment=VAlign_Top}}
  }
  SelectFXManagerLine {
   widgetClassPath=&HorizontalBox
   bIsVariable=true
   parent=VerticalBox_0
   widget=SelectFXManagerLine
   slot {class=VerticalBoxSlot, properties {verticalAlignment=VAlign_Center}}
  }
  FxManagersDropDown {
   widgetClassPath=/Script/FortniteEditor.SearchableStringComboBox
   bIsVariable=true
   parent=SelectFXManagerLine
   widget=FxManagersDropDown
   properties {
    font {
     bForceMonospaced=false
     bMaterialIsStencil=false
     fontMaterial=null
     fontObject=&Roboto
     letterSpacing=0
     monospacedWidth=1
     size=11.25
     skewAmount=0
     typefaceFontName=Bold
     outlineSettings {
      bApplyOutlineToDropShadows=false
      bMiteredCorners=false
      bSeparateFillAlpha=false
      outlineMaterial=null
      outlineSize=0
      outlineColor=<rgba 0 0 0 1>
     }
    }
    foregroundColor {colorUseRule=UseColor_Specified, specifiedColor=<rgba 0.5271 0.5271 0.5271 1>}
    itemStyle {
     bUseParentRowBrush=false
     activeBrush {
      bIsDynamicallyLoaded=false
      drawAs=Image
      imageType=NoImage
      mirroring=NoMirror
      resourceName=None
      resourceObject=null
      tiling=NoTile
      imageSize=<0 0>
      margin=<ltrb 0 0 0 0>
      outlineSettings {
       bUseBrushTransparency=false
       roundingType=HalfHeightRadius
       width=0
       color {colorUseRule=UseColor_Specified, specifiedColor=<rgba 0 0 0 0>}
       cornerRadii=<0 0 0 0>
      }
      tintColor {colorUseRule=UseColor_Specified, specifiedColor=<rgba 0 0.162 0.7454 1>}
      uVRegion {bIsValid=false, max=<0 0>, min=<0 0>}
     }
     activeHighlightedBrush {
      bIsDynamicallyLoaded=false
      drawAs=Image
      imageType=NoImage
      mirroring=NoMirror
      resourceName=None
      resourceObject=null
      tiling=NoTile
      imageSize=<0 0>
      margin=<ltrb 0 0 0 0>
      outlineSettings {
       bUseBrushTransparency=false
       roundingType=HalfHeightRadius
       width=0
       color {colorUseRule=UseColor_Specified, specifiedColor=<rgba 0 0 0 0>}
       cornerRadii=<0 0 0 0>
      }
      tintColor {colorUseRule=UseColor_Specified, specifiedColor=<rgba 0.0252 0.0319 0.0423 1>}
      uVRegion {bIsValid=false, max=<0 0>, min=<0 0>}
     }
     activeHoveredBrush {
      bIsDynamicallyLoaded=false
      drawAs=Image
      imageType=NoImage
      mirroring=NoMirror
      resourceName=None
      resourceObject=null
      tiling=NoTile
      imageSize=<0 0>
      margin=<ltrb 0 0 0 0>
      outlineSettings {
       bUseBrushTransparency=false
       roundingType=HalfHeightRadius
       width=0
       color {colorUseRule=UseColor_Specified, specifiedColor=<rgba 0 0 0 0>}
       cornerRadii=<0 0 0 0>
      }
      tintColor {colorUseRule=UseColor_Specified, specifiedColor=<rgba 0 0.162 0.7454 1>}
      uVRegion {bIsValid=false, max=<0 0>, min=<0 0>}
     }
     dropIndicator_Above {
      bIsDynamicallyLoaded=false
      drawAs=Box
      imageType=FullColor
      mirroring=NoMirror
      resourceName=../../../Engine/Content/Slate/Common/DropZoneIndicator_Above.png
      resourceObject=null
      tiling=NoTile
      imageSize=<0 0>
      margin=<ltrb 0.625 0.625 0 0>
      outlineSettings {
       bUseBrushTransparency=false
       roundingType=HalfHeightRadius
       width=0
       color {colorUseRule=UseColor_Specified, specifiedColor=<rgba 0 0 0 0>}
       cornerRadii=<0 0 0 0>
      }
      tintColor {colorUseRule=UseColor_Specified, specifiedColor=<rgba 0.0091 0.3515 0.9647 1>}
      uVRegion {bIsValid=false, max=<0 0>, min=<0 0>}
     }
     dropIndicator_Below {
      bIsDynamicallyLoaded=false
      drawAs=Box
      imageType=FullColor
      mirroring=NoMirror
      resourceName=../../../Engine/Content/Slate/Common/DropZoneIndicator_Below.png
      resourceObject=null
      tiling=NoTile
      imageSize=<0 0>
      margin=<ltrb 0.625 0 0 0.625>
      outlineSettings {
       bUseBrushTransparency=false
       roundingType=HalfHeightRadius
       width=0
       color {colorUseRule=UseColor_Specified, specifiedColor=<rgba 0 0 0 0>}
       cornerRadii=<0 0 0 0>
      }
      tintColor {colorUseRule=UseColor_Specified, specifiedColor=<rgba 0.0091 0.3515 0.9647 1>}
      uVRegion {bIsValid=false, max=<0 0>, min=<0 0>}
     }
     dropIndicator_Onto {
      bIsDynamicallyLoaded=false
      drawAs=Box
      imageType=FullColor
      mirroring=NoMirror
      resourceName=../../../Engine/Content/Slate/Common/DropZoneIndicator_Onto.png
      resourceObject=null
      tiling=NoTile
      imageSize=<0 0>
      margin=<ltrb 0.25 0.25 0.25 0.25>
      outlineSettings {
       bUseBrushTransparency=false
       roundingType=HalfHeightRadius
       width=0
       color {colorUseRule=UseColor_Specified, specifiedColor=<rgba 0 0 0 0>}
       cornerRadii=<0 0 0 0>
      }
      tintColor {colorUseRule=UseColor_Specified, specifiedColor=<rgba 0.0091 0.3515 0.9647 1>}
      uVRegion {bIsValid=false, max=<0 0>, min=<0 0>}
     }
     evenRowBackgroundBrush {
      bIsDynamicallyLoaded=false
      drawAs=Image
      imageType=NoImage
      mirroring=NoMirror
      resourceName=None
      resourceObject=null
      tiling=NoTile
      imageSize=<0 0>
      margin=<ltrb 0 0 0 0>
      outlineSettings {
       bUseBrushTransparency=false
       roundingType=HalfHeightRadius
       width=0
       color {colorUseRule=UseColor_Specified, specifiedColor=<rgba 0 0 0 0>}
       cornerRadii=<0 0 0 0>
      }
      tintColor {colorUseRule=UseColor_Specified, specifiedColor=<rgba 0.0103 0.0103 0.0103 1>}
      uVRegion {bIsValid=false, max=<0 0>, min=<0 0>}
     }
     evenRowBackgroundHoveredBrush {
      bIsDynamicallyLoaded=false
      drawAs=Image
      imageType=NoImage
      mirroring=NoMirror
      resourceName=None
      resourceObject=null
      tiling=NoTile
      imageSize=<0 0>
      margin=<ltrb 0 0 0 0>
      outlineSettings {
       bUseBrushTransparency=false
       roundingType=HalfHeightRadius
       width=0
       color {colorUseRule=UseColor_Specified, specifiedColor=<rgba 0 0 0 0>}
       cornerRadii=<0 0 0 0>
      }
      tintColor {colorUseRule=UseColor_Specified, specifiedColor=<rgba 0.0176 0.0176 0.0176 1>}
      uVRegion {bIsValid=false, max=<0 0>, min=<0 0>}
     }
     inactiveBrush {
      bIsDynamicallyLoaded=false
      drawAs=Image
      imageType=NoImage
      mirroring=NoMirror
      resourceName=None
      resourceObject=null
      tiling=NoTile
      imageSize=<0 0>
      margin=<ltrb 0 0 0 0>
      outlineSettings {
       bUseBrushTransparency=false
       roundingType=HalfHeightRadius
       width=0
       color {colorUseRule=UseColor_Specified, specifiedColor=<rgba 0 0 0 0>}
       cornerRadii=<0 0 0 0>
      }
      tintColor {colorUseRule=UseColor_Specified, specifiedColor=<rgba 0.0513 0.0953 0.159 1>}
      uVRegion {bIsValid=false, max=<0 0>, min=<0 0>}
     }
     inactiveHighlightedBrush {
      bIsDynamicallyLoaded=false
      drawAs=Image
      imageType=NoImage
      mirroring=NoMirror
      resourceName=None
      resourceObject=null
      tiling=NoTile
      imageSize=<0 0>
      margin=<ltrb 0 0 0 0>
      outlineSettings {
       bUseBrushTransparency=false
       roundingType=HalfHeightRadius
       width=0
       color {colorUseRule=UseColor_Specified, specifiedColor=<rgba 0 0 0 0>}
       cornerRadii=<0 0 0 0>
      }
      tintColor {colorUseRule=UseColor_Specified, specifiedColor=<rgba 0.0252 0.0319 0.0423 1>}
      uVRegion {bIsValid=false, max=<0 0>, min=<0 0>}
     }
     inactiveHoveredBrush {
      bIsDynamicallyLoaded=false
      drawAs=Image
      imageType=NoImage
      mirroring=NoMirror
      resourceName=None
      resourceObject=null
      tiling=NoTile
      imageSize=<0 0>
      margin=<ltrb 0 0 0 0>
      outlineSettings {
       bUseBrushTransparency=false
       roundingType=HalfHeightRadius
       width=0
       color {colorUseRule=UseColor_Specified, specifiedColor=<rgba 0 0 0 0>}
       cornerRadii=<0 0 0 0>
      }
      tintColor {colorUseRule=UseColor_Specified, specifiedColor=<rgba 0.0513 0.0953 0.159 1>}
      uVRegion {bIsValid=false, max=<0 0>, min=<0 0>}
     }
     oddRowBackgroundBrush {
      bIsDynamicallyLoaded=false
      drawAs=Image
      imageType=NoImage
      mirroring=NoMirror
      resourceName=None
      resourceObject=null
      tiling=NoTile
      imageSize=<0 0>
      margin=<ltrb 0 0 0 0>
      outlineSettings {
       bUseBrushTransparency=false
       roundingType=HalfHeightRadius
       width=0
       color {colorUseRule=UseColor_Specified, specifiedColor=<rgba 0 0 0 0>}
       cornerRadii=<0 0 0 0>
      }
      tintColor {colorUseRule=UseColor_Specified, specifiedColor=<rgba 0.0103 0.0103 0.0103 1>}
      uVRegion {bIsValid=false, max=<0 0>, min=<0 0>}
     }
     oddRowBackgroundHoveredBrush {
      bIsDynamicallyLoaded=false
      drawAs=Image
      imageType=NoImage
      mirroring=NoMirror
      resourceName=None
      resourceObject=null
      tiling=NoTile
      imageSize=<0 0>
      margin=<ltrb 0 0 0 0>
      outlineSettings {
       bUseBrushTransparency=false
       roundingType=HalfHeightRadius
       width=0
       color {colorUseRule=UseColor_Specified, specifiedColor=<rgba 0 0 0 0>}
       cornerRadii=<0 0 0 0>
      }
      tintColor {colorUseRule=UseColor_Specified, specifiedColor=<rgba 0.0176 0.0176 0.0176 1>}
      uVRegion {bIsValid=false, max=<0 0>, min=<0 0>}
     }
     parentRowBackgroundBrush {
      bIsDynamicallyLoaded=false
      drawAs=Image
      imageType=NoImage
      mirroring=NoMirror
      resourceName=None
      resourceObject=null
      tiling=NoTile
      imageSize=<32 32>
      margin=<ltrb 0 0 0 0>
      outlineSettings {
       bUseBrushTransparency=false
       roundingType=HalfHeightRadius
       width=0
       color {colorUseRule=UseColor_Specified, specifiedColor=<rgba 0 0 0 0>}
       cornerRadii=<0 0 0 0>
      }
      tintColor {colorUseRule=UseColor_Specified, specifiedColor=<rgba 1 1 1 1>}
      uVRegion {bIsValid=false, max=<0 0>, min=<0 0>}
     }
     parentRowBackgroundHoveredBrush {
      bIsDynamicallyLoaded=false
      drawAs=Image
      imageType=NoImage
      mirroring=NoMirror
      resourceName=None
      resourceObject=null
      tiling=NoTile
      imageSize=<32 32>
      margin=<ltrb 0 0 0 0>
      outlineSettings {
       bUseBrushTransparency=false
       roundingType=HalfHeightRadius
       width=0
       color {colorUseRule=UseColor_Specified, specifiedColor=<rgba 0 0 0 0>}
       cornerRadii=<0 0 0 0>
      }
      tintColor {colorUseRule=UseColor_Specified, specifiedColor=<rgba 1 1 1 1>}
      uVRegion {bIsValid=false, max=<0 0>, min=<0 0>}
     }
     selectedTextColor {colorUseRule=UseColor_Specified, specifiedColor=<rgba 1 1 1 1>}
     selectorFocusedBrush {
      bIsDynamicallyLoaded=false
      drawAs=Border
      imageType=FullColor
      mirroring=NoMirror
      resourceName=../../../Engine/Content/Slate/Common/Selector.png
      resourceObject=null
      tiling=Both
      imageSize=<0 0>
      margin=<ltrb 0.25 0.25 0.25 0.25>
      outlineSettings {
       bUseBrushTransparency=false
       roundingType=HalfHeightRadius
       width=0
       color {colorUseRule=UseColor_Specified, specifiedColor=<rgba 0 0 0 0>}
       cornerRadii=<0 0 0 0>
      }
      tintColor {colorUseRule=UseColor_Specified, specifiedColor=<rgba 0.701 0.225 0.003 1>}
      uVRegion {bIsValid=false, max=<0 0>, min=<0 0>}
     }
     textColor {colorUseRule=UseColor_Specified, specifiedColor=<rgba 0.5271 0.5271 0.5271 1>}
    }
    widgetStyle {
     comboButtonStyle {
      downArrowAlign=VAlign_Center
      buttonStyle {
       clickedSlateSound {resourceObject=null}
       disabled {
        bIsDynamicallyLoaded=false
        drawAs=RoundedBox
        imageType=NoImage
        mirroring=NoMirror
        resourceName=None
        resourceObject=null
        tiling=NoTile
        imageSize=<0 0>
        margin=<ltrb 0 0 0 0>
        outlineSettings {
         bUseBrushTransparency=false
         roundingType=FixedRadius
         width=1
         color {colorUseRule=UseColor_ColorTable, specifiedColor=<rgba 0 0 0 0>}
         cornerRadii=<4 4 4 4>
        }
        tintColor {colorUseRule=UseColor_Specified, specifiedColor=<rgba 0.0395 0.0395 0.0395 1>}
        uVRegion {bIsValid=false, max=<0 0>, min=<0 0>}
       }
       disabledForeground {colorUseRule=UseColor_ColorTable, specifiedColor=<rgba 0 0 0 0>}
       hovered {
        bIsDynamicallyLoaded=false
        drawAs=RoundedBox
        imageType=NoImage
        mirroring=NoMirror
        resourceName=None
        resourceObject=null
        tiling=NoTile
        imageSize=<0 0>
        margin=<ltrb 0 0 0 0>
        outlineSettings {
         bUseBrushTransparency=false
         roundingType=FixedRadius
         width=1
         color {colorUseRule=UseColor_ColorTable, specifiedColor=<rgba 0 0 0 0>}
         cornerRadii=<4 4 4 4>
        }
        tintColor {colorUseRule=UseColor_Specified, specifiedColor=<rgba 0.0048 0.0048 0.0048 1>}
        uVRegion {bIsValid=false, max=<0 0>, min=<0 0>}
       }
       hoveredForeground {colorUseRule=UseColor_ColorTable, specifiedColor=<rgba 0 0 0 0>}
       hoveredSlateSound {resourceObject=null}
       normal {
        bIsDynamicallyLoaded=false
        drawAs=RoundedBox
        imageType=NoImage
        mirroring=NoMirror
        resourceName=None
        resourceObject=null
        tiling=NoTile
        imageSize=<0 0>
        margin=<ltrb 0 0 0 0>
        outlineSettings {
         bUseBrushTransparency=false
         roundingType=FixedRadius
         width=1
         color {colorUseRule=UseColor_ColorTable, specifiedColor=<rgba 0 0 0 0>}
         cornerRadii=<4 4 4 4>
        }
        tintColor {colorUseRule=UseColor_Specified, specifiedColor=<rgba 0.0048 0.0048 0.0048 1>}
        uVRegion {bIsValid=false, max=<0 0>, min=<0 0>}
       }
       normalForeground {colorUseRule=UseColor_ColorTable, specifiedColor=<rgba 0 0 0 0>}
       normalPadding=<ltrb 8 1 8 1>
       pressed {
        bIsDynamicallyLoaded=false
        drawAs=RoundedBox
        imageType=NoImage
        mirroring=NoMirror
        resourceName=None
        resourceObject=null
        tiling=NoTile
        imageSize=<0 0>
        margin=<ltrb 0 0 0 0>
        outlineSettings {
         bUseBrushTransparency=false
         roundingType=FixedRadius
         width=1
         color {colorUseRule=UseColor_ColorTable, specifiedColor=<rgba 0 0 0 0>}
         cornerRadii=<4 4 4 4>
        }
        tintColor {colorUseRule=UseColor_Specified, specifiedColor=<rgba 0.0048 0.0048 0.0048 1>}
        uVRegion {bIsValid=false, max=<0 0>, min=<0 0>}
       }
       pressedForeground {colorUseRule=UseColor_ColorTable, specifiedColor=<rgba 0 0 0 0>}
       pressedPadding=<ltrb 8 2 8 0>
       pressedSlateSound {resourceObject=null}
      }
      contentPadding=<ltrb 0 0 0 0>
      downArrowImage {
       bIsDynamicallyLoaded=false
       drawAs=Image
       imageType=Vector
       mirroring=NoMirror
       resourceName=../../../Engine/Content/Slate/Starship/CoreWidgets/ComboBox/wide-chevron-down.svg
       resourceObject=null
       tiling=NoTile
       imageSize=<16 16>
       margin=<ltrb 0 0 0 0>
       outlineSettings {
        bUseBrushTransparency=false
        roundingType=HalfHeightRadius
        width=0
        color {colorUseRule=UseColor_Specified, specifiedColor=<rgba 0 0 0 0>}
        cornerRadii=<0 0 0 0>
       }
       tintColor {colorUseRule=UseColor_Specified, specifiedColor=<rgba 1 1 1 1>}
       uVRegion {bIsValid=false, max=<0 0>, min=<0 0>}
      }
      downArrowPadding=<ltrb 2 2 2 2>
      menuBorderBrush {
       bIsDynamicallyLoaded=false
       drawAs=RoundedBox
       imageType=NoImage
       mirroring=NoMirror
       resourceName=None
       resourceObject=null
       tiling=NoTile
       imageSize=<0 0>
       margin=<ltrb 0 0 0 0>
       outlineSettings {
        bUseBrushTransparency=false
        roundingType=FixedRadius
        width=1
        color {colorUseRule=UseColor_ColorTable, specifiedColor=<rgba 0 0 0 0>}
        cornerRadii=<0 0 0 0>
       }
       tintColor {colorUseRule=UseColor_Specified, specifiedColor=<rgba 0.0395 0.0395 0.0395 1>}
       uVRegion {bIsValid=false, max=<0 0>, min=<0 0>}
      }
      menuBorderPadding=<ltrb 0 0 0 0>
      shadowColorAndOpacity=<rgba 0 0 0 1>
      shadowOffset=<0 0>
     }
     contentPadding=<ltrb 0 0 0 0>
     menuRowPadding=<ltrb 8 3 8 3>
     pressedSlateSound {resourceObject=null}
     selectionChangeSlateSound {resourceObject=null}
    }
   }
   slot {class=HorizontalBoxSlot, properties {padding=<ltrb 0 5 0 5>, size {sizeRule=Fill, value=1.5}}}
  }
  Spacer_65 {widgetClassPath=&Spacer, parent=SelectFXManagerLine, widget=Spacer_65, properties {size=<15 1>}, slot {class=HorizontalBoxSlot}}
  RefreshFxManagersListButton {
   widgetClassPath=&EditorUtilityButton
   parent=SelectFXManagerLine
   widget=RefreshFxManagersListButton
   properties {toolTipText=Refresh the list if there is a new FX Manager.}
   slot {class=HorizontalBoxSlot, properties {size {sizeRule=Fill, value=1}}}
  }
  RefreshText {
   widgetClassPath=&TextBlock
   parent=RefreshFxManagersListButton
   widget=RefreshText
   properties {
    autoWrapText=true
    justification=Center
    text=Refresh List
    visibility=SelfHitTestInvisible
    wrappingPolicy=AllowPerCharacterWrapping
    font {
     bForceMonospaced=false
     bMaterialIsStencil=false
     fontMaterial=null
     fontObject=&Roboto
     letterSpacing=0
     monospacedWidth=1
     size=10.5
     skewAmount=0
     typefaceFontName=Bold
     outlineSettings {
      bApplyOutlineToDropShadows=false
      bMiteredCorners=false
      bSeparateFillAlpha=false
      outlineMaterial=null
      outlineSize=0
      outlineColor=<rgba 0 0 0 1>
     }
    }
   }
   slot {class=ButtonSlot, properties {horizontalAlignment=HAlign_Fill}}
  }
  Spacer {widgetClassPath=&Spacer, parent=SelectFXManagerLine, widget=Spacer, properties {size=<10 1>}, slot {class=HorizontalBoxSlot}}
  CreateNewFxManagerButton {
   widgetClassPath=&EditorUtilityButton
   parent=SelectFXManagerLine
   widget=CreateNewFxManagerButton
   properties {toolTipText=Open a new option in window to support you to create a new FXManager.}
   slot {class=HorizontalBoxSlot, properties {size {sizeRule=Fill, value=1}}}
  }
  CreateNewFxManagerText {
   widgetClassPath=&TextBlock
   parent=CreateNewFxManagerButton
   widget=CreateNewFxManagerText
   properties {
    autoWrapText=true
    justification=Center
    text=Create New FX Manager
    visibility=SelfHitTestInvisible
    wrappingPolicy=AllowPerCharacterWrapping
    font {
     bForceMonospaced=false
     bMaterialIsStencil=false
     fontMaterial=null
     fontObject=&Roboto
     letterSpacing=0
     monospacedWidth=1
     size=10.5
     skewAmount=0
     typefaceFontName=Bold
     outlineSettings {
      bApplyOutlineToDropShadows=false
      bMiteredCorners=false
      bSeparateFillAlpha=false
      outlineMaterial=null
      outlineSize=0
      outlineColor=<rgba 0 0 0 1>
     }
    }
   }
   slot {class=ButtonSlot, properties {horizontalAlignment=HAlign_Fill}}
  }
  Spacer_1 {widgetClassPath=&Spacer, parent=SelectFXManagerLine, widget=Spacer_1, properties {size=<10 1>}, slot {class=HorizontalBoxSlot}}
  AddSelectedActorButton {
   widgetClassPath=&EditorUtilityButton
   parent=SelectFXManagerLine
   widget=AddSelectedActorButton
   properties {toolTipText=Add valid selected actor to the current selected fx manager from list}
   slot {class=HorizontalBoxSlot, properties {size {sizeRule=Fill, value=1}}}
  }
  AddSelectedActorText {
   widgetClassPath=&TextBlock
   bIsVariable=true
   parent=AddSelectedActorButton
   widget=AddSelectedActorText
   properties {
    autoWrapText=true
    justification=Center
    text=Add Selected Actor to Manager
    visibility=SelfHitTestInvisible
    font {
     bForceMonospaced=false
     bMaterialIsStencil=false
     fontMaterial=null
     fontObject=&Roboto
     letterSpacing=0
     monospacedWidth=1
     size=10.5
     skewAmount=0
     typefaceFontName=Bold
     outlineSettings {
      bApplyOutlineToDropShadows=false
      bMiteredCorners=false
      bSeparateFillAlpha=false
      outlineMaterial=null
      outlineSize=0
      outlineColor=<rgba 0 0 0 1>
     }
    }
   }
   slot {class=ButtonSlot, properties {horizontalAlignment=HAlign_Fill}}
  }
  CreateFxManagerLine {
   widgetClassPath=&HorizontalBox
   bIsVariable=true
   parent=VerticalBox_0
   widget=CreateFxManagerLine
   properties {visibility=Collapsed}
   slot {class=VerticalBoxSlot, properties {verticalAlignment=VAlign_Center}}
  }
  SizeBox_1 {
   widgetClassPath=&SizeBox
   parent=CreateFxManagerLine
   widget=SizeBox_1
   properties {bOverride_MinDesiredHeight=true, minDesiredHeight=30}
   slot {class=HorizontalBoxSlot, properties {size {sizeRule=Fill, value=1}}}
  }
  NewFxManagerNameTextBox {
   widgetClassPath=&EditorUtilityEditableTextBox
   bIsVariable=true
   parent=SizeBox_1
   widget=NewFxManagerNameTextBox
   properties {
    hintText=Name Of New FXManager
    widgetStyle {
     backgroundColor {colorUseRule=UseColor_Specified, specifiedColor=<rgba 1 1 1 1>}
     backgroundImageFocused {
      bIsDynamicallyLoaded=false
      drawAs=RoundedBox
      imageType=NoImage
      mirroring=NoMirror
      resourceName=None
      resourceObject=null
      tiling=NoTile
      imageSize=<0 0>
      margin=<ltrb 0 0 0 0>
      outlineSettings {
       bUseBrushTransparency=false
       roundingType=FixedRadius
       width=1
       color {colorUseRule=UseColor_ColorTable, specifiedColor=<rgba 0 0 0 0>}
       cornerRadii=<4 4 4 4>
      }
      tintColor {colorUseRule=UseColor_Specified, specifiedColor=<rgba 0.0048 0.0048 0.0048 1>}
      uVRegion {bIsValid=false, max=<0 0>, min=<0 0>}
     }
     backgroundImageHovered {
      bIsDynamicallyLoaded=false
      drawAs=RoundedBox
      imageType=NoImage
      mirroring=NoMirror
      resourceName=None
      resourceObject=null
      tiling=NoTile
      imageSize=<0 0>
      margin=<ltrb 0 0 0 0>
      outlineSettings {
       bUseBrushTransparency=false
       roundingType=FixedRadius
       width=1
       color {colorUseRule=UseColor_ColorTable, specifiedColor=<rgba 0 0 0 0>}
       cornerRadii=<4 4 4 4>
      }
      tintColor {colorUseRule=UseColor_Specified, specifiedColor=<rgba 0.0048 0.0048 0.0048 1>}
      uVRegion {bIsValid=false, max=<0 0>, min=<0 0>}
     }
     backgroundImageNormal {
      bIsDynamicallyLoaded=false
      drawAs=RoundedBox
      imageType=NoImage
      mirroring=NoMirror
      resourceName=None
      resourceObject=null
      tiling=NoTile
      imageSize=<0 0>
      margin=<ltrb 0 0 0 0>
      outlineSettings {
       bUseBrushTransparency=false
       roundingType=FixedRadius
       width=1
       color {colorUseRule=UseColor_ColorTable, specifiedColor=<rgba 0 0 0 0>}
       cornerRadii=<4 4 4 4>
      }
      tintColor {colorUseRule=UseColor_Specified, specifiedColor=<rgba 0.0048 0.0048 0.0048 1>}
      uVRegion {bIsValid=false, max=<0 0>, min=<0 0>}
     }
     backgroundImageReadOnly {
      bIsDynamicallyLoaded=false
      drawAs=RoundedBox
      imageType=NoImage
      mirroring=NoMirror
      resourceName=None
      resourceObject=null
      tiling=NoTile
      imageSize=<0 0>
      margin=<ltrb 0 0 0 0>
      outlineSettings {
       bUseBrushTransparency=false
       roundingType=FixedRadius
       width=1
       color {colorUseRule=UseColor_ColorTable, specifiedColor=<rgba 0 0 0 0>}
       cornerRadii=<4 4 4 4>
      }
      tintColor {colorUseRule=UseColor_Specified, specifiedColor=<rgba 0.0284 0.0284 0.0284 1>}
      uVRegion {bIsValid=false, max=<0 0>, min=<0 0>}
     }
     focusedForegroundColor {colorUseRule=UseColor_Specified, specifiedColor=<rgba 1 1 1 1>}
     foregroundColor {colorUseRule=UseColor_Specified, specifiedColor=<rgba 0.5271 0.5271 0.5271 1>}
     hScrollBarPadding=<ltrb 0 0 0 0>
     padding=<ltrb 8 3 8 4>
     readOnlyForegroundColor {colorUseRule=UseColor_Foreground, specifiedColor=<rgba 1 0 1 1>}
     scrollBarStyle {
      thickness=8
      draggedThumbImage {
       bIsDynamicallyLoaded=false
       drawAs=RoundedBox
       imageType=NoImage
       mirroring=NoMirror
       resourceName=None
       resourceObject=null
       tiling=NoTile
       imageSize=<0 0>
       margin=<ltrb 0 0 0 0>
       outlineSettings {
        bUseBrushTransparency=false
        roundingType=FixedRadius
        width=0
        color {colorUseRule=UseColor_Specified, specifiedColor=<rgba 0 0 0 0>}
        cornerRadii=<4 4 4 4>
       }
       tintColor {colorUseRule=UseColor_Specified, specifiedColor=<rgba 0.2159 0.2159 0.2159 1>}
       uVRegion {bIsValid=false, max=<0 0>, min=<0 0>}
      }
      horizontalBackgroundImage {
       bIsDynamicallyLoaded=false
       drawAs=NoDrawType
       imageType=NoImage
       mirroring=NoMirror
       resourceName=None
       resourceObject=null
       tiling=NoTile
       imageSize=<0 0>
       margin=<ltrb 0 0 0 0>
       outlineSettings {
        bUseBrushTransparency=false
        roundingType=HalfHeightRadius
        width=0
        color {colorUseRule=UseColor_Specified, specifiedColor=<rgba 0 0 0 0>}
        cornerRadii=<0 0 0 0>
       }
       tintColor {colorUseRule=UseColor_Specified, specifiedColor=<rgba 1 1 1 1>}
       uVRegion {bIsValid=false, max=<0 0>, min=<0 0>}
      }
      horizontalBottomSlotImage {
       bIsDynamicallyLoaded=false
       drawAs=NoDrawType
       imageType=NoImage
       mirroring=NoMirror
       resourceName=None
       resourceObject=null
       tiling=NoTile
       imageSize=<0 0>
       margin=<ltrb 0 0 0 0>
       outlineSettings {
        bUseBrushTransparency=false
        roundingType=HalfHeightRadius
        width=0
        color {colorUseRule=UseColor_Specified, specifiedColor=<rgba 0 0 0 0>}
        cornerRadii=<0 0 0 0>
       }
       tintColor {colorUseRule=UseColor_Specified, specifiedColor=<rgba 1 1 1 1>}
       uVRegion {bIsValid=false, max=<0 0>, min=<0 0>}
      }
      horizontalTopSlotImage {
       bIsDynamicallyLoaded=false
       drawAs=NoDrawType
       imageType=NoImage
       mirroring=NoMirror
       resourceName=None
       resourceObject=null
       tiling=NoTile
       imageSize=<0 0>
       margin=<ltrb 0 0 0 0>
       outlineSettings {
        bUseBrushTransparency=false
        roundingType=HalfHeightRadius
        width=0
        color {colorUseRule=UseColor_Specified, specifiedColor=<rgba 0 0 0 0>}
        cornerRadii=<0 0 0 0>
       }
       tintColor {colorUseRule=UseColor_Specified, specifiedColor=<rgba 1 1 1 1>}
       uVRegion {bIsValid=false, max=<0 0>, min=<0 0>}
      }
      hoveredThumbImage {
       bIsDynamicallyLoaded=false
       drawAs=RoundedBox
       imageType=NoImage
       mirroring=NoMirror
       resourceName=None
       resourceObject=null
       tiling=NoTile
       imageSize=<0 0>
       margin=<ltrb 0 0 0 0>
       outlineSettings {
        bUseBrushTransparency=false
        roundingType=FixedRadius
        width=0
        color {colorUseRule=UseColor_Specified, specifiedColor=<rgba 0 0 0 0>}
        cornerRadii=<4 4 4 4>
       }
       tintColor {colorUseRule=UseColor_Specified, specifiedColor=<rgba 0.2159 0.2159 0.2159 1>}
       uVRegion {bIsValid=false, max=<0 0>, min=<0 0>}
      }
      normalThumbImage {
       bIsDynamicallyLoaded=false
       drawAs=RoundedBox
       imageType=NoImage
       mirroring=NoMirror
       resourceName=None
       resourceObject=null
       tiling=NoTile
       imageSize=<0 0>
       margin=<ltrb 0 0 0 0>
       outlineSettings {
        bUseBrushTransparency=false
        roundingType=FixedRadius
        width=0
        color {colorUseRule=UseColor_Specified, specifiedColor=<rgba 0 0 0 0>}
        cornerRadii=<4 4 4 4>
       }
       tintColor {colorUseRule=UseColor_Specified, specifiedColor=<rgba 0.0953 0.0953 0.0953 1>}
       uVRegion {bIsValid=false, max=<0 0>, min=<0 0>}
      }
      verticalBackgroundImage {
       bIsDynamicallyLoaded=false
       drawAs=NoDrawType
       imageType=NoImage
       mirroring=NoMirror
       resourceName=None
       resourceObject=null
       tiling=NoTile
       imageSize=<0 0>
       margin=<ltrb 0 0 0 0>
       outlineSettings {
        bUseBrushTransparency=false
        roundingType=HalfHeightRadius
        width=0
        color {colorUseRule=UseColor_Specified, specifiedColor=<rgba 0 0 0 0>}
        cornerRadii=<0 0 0 0>
       }
       tintColor {colorUseRule=UseColor_Specified, specifiedColor=<rgba 1 1 1 1>}
       uVRegion {bIsValid=false, max=<0 0>, min=<0 0>}
      }
      verticalBottomSlotImage {
       bIsDynamicallyLoaded=false
       drawAs=NoDrawType
       imageType=NoImage
       mirroring=NoMirror
       resourceName=None
       resourceObject=null
       tiling=NoTile
       imageSize=<0 0>
       margin=<ltrb 0 0 0 0>
       outlineSettings {
        bUseBrushTransparency=false
        roundingType=HalfHeightRadius
        width=0
        color {colorUseRule=UseColor_Specified, specifiedColor=<rgba 0 0 0 0>}
        cornerRadii=<0 0 0 0>
       }
       tintColor {colorUseRule=UseColor_Specified, specifiedColor=<rgba 1 1 1 1>}
       uVRegion {bIsValid=false, max=<0 0>, min=<0 0>}
      }
      verticalTopSlotImage {
       bIsDynamicallyLoaded=false
       drawAs=NoDrawType
       imageType=NoImage
       mirroring=NoMirror
       resourceName=None
       resourceObject=null
       tiling=NoTile
       imageSize=<0 0>
       margin=<ltrb 0 0 0 0>
       outlineSettings {
        bUseBrushTransparency=false
        roundingType=HalfHeightRadius
        width=0
        color {colorUseRule=UseColor_Specified, specifiedColor=<rgba 0 0 0 0>}
        cornerRadii=<0 0 0 0>
       }
       tintColor {colorUseRule=UseColor_Specified, specifiedColor=<rgba 1 1 1 1>}
       uVRegion {bIsValid=false, max=<0 0>, min=<0 0>}
      }
     }
     textStyle {
      overflowPolicy=Clip
      transformPolicy=None
      colorAndOpacity {colorUseRule=UseColor_Foreground, specifiedColor=<rgba 1 0 1 1>}
      font {
       bForceMonospaced=false
       bMaterialIsStencil=false
       fontMaterial=null
       fontObject=&HeadingNow
       letterSpacing=0
       monospacedWidth=1
       size=11.25
       skewAmount=0
       typefaceFontName=76 Bold
       outlineSettings {
        bApplyOutlineToDropShadows=false
        bMiteredCorners=false
        bSeparateFillAlpha=false
        outlineMaterial=null
        outlineSize=0
        outlineColor=<rgba 0 0 0 1>
       }
      }
      highlightColor {colorUseRule=UseColor_ColorTable, specifiedColor=<rgba 0 0 0 0>}
      highlightShape {
       bIsDynamicallyLoaded=false
       drawAs=Image
       imageType=NoImage
       mirroring=NoMirror
       resourceName=None
       resourceObject=null
       tiling=NoTile
       imageSize=<0 0>
       margin=<ltrb 0 0 0 0>
       outlineSettings {
        bUseBrushTransparency=false
        roundingType=HalfHeightRadius
        width=0
        color {colorUseRule=UseColor_Specified, specifiedColor=<rgba 0 0 0 0>}
        cornerRadii=<0 0 0 0>
       }
       tintColor {colorUseRule=UseColor_ColorTable, specifiedColor=<rgba 0 0 0 0>}
       uVRegion {bIsValid=false, max=<0 0>, min=<0 0>}
      }
      selectedBackgroundColor {colorUseRule=UseColor_ColorTable, specifiedColor=<rgba 0 0 0 0>}
      shadowColorAndOpacity=<rgba 0 0 0 1>
      shadowOffset=<0 0>
      strikeBrush {
       bIsDynamicallyLoaded=false
       drawAs=Image
       imageType=NoImage
       mirroring=NoMirror
       resourceName=None
       resourceObject=null
       tiling=NoTile
       imageSize=<32 32>
       margin=<ltrb 0 0 0 0>
       outlineSettings {
        bUseBrushTransparency=false
        roundingType=HalfHeightRadius
        width=0
        color {colorUseRule=UseColor_Specified, specifiedColor=<rgba 0 0 0 0>}
        cornerRadii=<0 0 0 0>
       }
       tintColor {colorUseRule=UseColor_Specified, specifiedColor=<rgba 1 1 1 1>}
       uVRegion {bIsValid=false, max=<0 0>, min=<0 0>}
      }
      underlineBrush {
       bIsDynamicallyLoaded=false
       drawAs=Image
       imageType=NoImage
       mirroring=NoMirror
       resourceName=None
       resourceObject=null
       tiling=NoTile
       imageSize=<32 32>
       margin=<ltrb 0 0 0 0>
       outlineSettings {
        bUseBrushTransparency=false
        roundingType=HalfHeightRadius
        width=0
        color {colorUseRule=UseColor_Specified, specifiedColor=<rgba 0 0 0 0>}
        cornerRadii=<0 0 0 0>
       }
       tintColor {colorUseRule=UseColor_Specified, specifiedColor=<rgba 1 1 1 1>}
       uVRegion {bIsValid=false, max=<0 0>, min=<0 0>}
      }
     }
     vScrollBarPadding=<ltrb 0 0 0 0>
    }
   }
   slot {class=SizeBoxSlot}
  }
  Spacer_2 {widgetClassPath=&Spacer, parent=CreateFxManagerLine, widget=Spacer_2, properties {size=<15 1>}, slot {class=HorizontalBoxSlot}}
  CreateFxButton {
   widgetClassPath=&EditorUtilityButton
   parent=CreateFxManagerLine
   widget=CreateFxButton
   properties {toolTipText=Create a New FxManager}
   slot {class=HorizontalBoxSlot, properties {size {sizeRule=Fill, value=1}}}
  }
  CreateFxText {
   widgetClassPath=&TextBlock
   parent=CreateFxButton
   widget=CreateFxText
   properties {
    autoWrapText=true
    justification=Center
    text=Create
    visibility=SelfHitTestInvisible
    wrappingPolicy=AllowPerCharacterWrapping
    font {
     bForceMonospaced=false
     bMaterialIsStencil=false
     fontMaterial=null
     fontObject=&Roboto
     letterSpacing=0
     monospacedWidth=1
     size=10.5
     skewAmount=0
     typefaceFontName=Bold
     outlineSettings {
      bApplyOutlineToDropShadows=false
      bMiteredCorners=false
      bSeparateFillAlpha=false
      outlineMaterial=null
      outlineSize=0
      outlineColor=<rgba 0 0 0 1>
     }
    }
   }
   slot {class=ButtonSlot, properties {horizontalAlignment=HAlign_Fill}}
  }
  Spacer_3 {widgetClassPath=&Spacer, parent=CreateFxManagerLine, widget=Spacer_3, properties {size=<10 1>}, slot {class=HorizontalBoxSlot}}
  CreateAndAddButton {
   widgetClassPath=&EditorUtilityButton
   parent=CreateFxManagerLine
   widget=CreateAndAddButton
   slot {class=HorizontalBoxSlot, properties {size {sizeRule=Fill, value=1}}}
  }
  CrateAndAddText {
   widgetClassPath=&TextBlock
   bIsVariable=true
   parent=CreateAndAddButton
   widget=CrateAndAddText
   properties {
    autoWrapText=true
    justification=Center
    text=Create and Add
    visibility=SelfHitTestInvisible
    wrappingPolicy=AllowPerCharacterWrapping
    font {
     bForceMonospaced=false
     bMaterialIsStencil=false
     fontMaterial=null
     fontObject=&Roboto
     letterSpacing=0
     monospacedWidth=1
     size=10.5
     skewAmount=0
     typefaceFontName=Bold
     outlineSettings {
      bApplyOutlineToDropShadows=false
      bMiteredCorners=false
      bSeparateFillAlpha=false
      outlineMaterial=null
      outlineSize=0
      outlineColor=<rgba 0 0 0 1>
     }
    }
   }
   slot {class=ButtonSlot, properties {horizontalAlignment=HAlign_Fill}}
  }
  Spacer_5 {widgetClassPath=&Spacer, parent=CreateFxManagerLine, widget=Spacer_5, properties {size=<10 1>}, slot {class=HorizontalBoxSlot}}
  CancelFxManButton {
   widgetClassPath=&EditorUtilityButton
   parent=CreateFxManagerLine
   widget=CancelFxManButton
   slot {class=HorizontalBoxSlot, properties {size {sizeRule=Fill, value=1}}}
  }
  CancelFxManText {
   widgetClassPath=&TextBlock
   bIsVariable=true
   parent=CancelFxManButton
   widget=CancelFxManText
   properties {
    autoWrapText=true
    justification=Center
    text=Cancel
    visibility=SelfHitTestInvisible
    wrappingPolicy=AllowPerCharacterWrapping
    font {
     bForceMonospaced=false
     bMaterialIsStencil=false
     fontMaterial=null
     fontObject=&Roboto
     letterSpacing=0
     monospacedWidth=1
     size=10.5
     skewAmount=0
     typefaceFontName=Bold
     outlineSettings {
      bApplyOutlineToDropShadows=false
      bMiteredCorners=false
      bSeparateFillAlpha=false
      outlineMaterial=null
      outlineSize=0
      outlineColor=<rgba 0 0 0 1>
     }
    }
   }
   slot {class=ButtonSlot, properties {horizontalAlignment=HAlign_Fill}}
  }
  Border_0 {
   widgetClassPath=&Border
   parent=VerticalBox_0
   widget=Border_0
   properties {
    verticalAlignment=VAlign_Center
    background {
     bIsDynamicallyLoaded=false
     drawAs=NoDrawType
     imageType=NoImage
     mirroring=NoMirror
     resourceName=None
     resourceObject=null
     tiling=NoTile
     imageSize=<32 32>
     margin=<ltrb 0 0 0 0>
     outlineSettings {
      bUseBrushTransparency=false
      roundingType=HalfHeightRadius
      width=0
      color {colorUseRule=UseColor_Specified, specifiedColor=<rgba 0 0 0 0>}
      cornerRadii=<0 0 0 0>
     }
     tintColor {colorUseRule=UseColor_Specified, specifiedColor=<rgba 1 1 1 1>}
     uVRegion {bIsValid=false, max=<0 0>, min=<0 0>}
    }
    padding=<ltrb 4 25 4 5>
   }
   slot {class=VerticalBoxSlot}
  }
  Line {
   widgetClassPath=&Image
   parent=Border_0
   widget=Line
   properties {
    brush {
     bIsDynamicallyLoaded=false
     drawAs=Image
     imageType=NoImage
     mirroring=NoMirror
     resourceName=None
     resourceObject=null
     tiling=NoTile
     imageSize=<32 3>
     margin=<ltrb 0 0 0 0>
     outlineSettings {
      bUseBrushTransparency=false
      roundingType=HalfHeightRadius
      width=0
      color {colorUseRule=UseColor_Specified, specifiedColor=<rgba 0 0 0 0>}
      cornerRadii=<0 0 0 0>
     }
     tintColor {colorUseRule=UseColor_Specified, specifiedColor=<rgba 1 1 1 1>}
     uVRegion {bIsValid=false, max=<0 0>, min=<0 0>}
    }
    colorAndOpacity=<rgba 0.0365 0.0365 0.0365 1>
   }
   slot {class=BorderSlot, properties {verticalAlignment=VAlign_Center, padding=<ltrb 4 25 4 5>}}
  }
  FxManagerDetailsScrollBox {
   widgetClassPath=&EditorUtilityScrollBox
   bIsVariable=true
   parent=VerticalBox_0
   widget=FxManagerDetailsScrollBox
   properties {widgetBarStyle {
     thickness=5
     draggedThumbImage {
      bIsDynamicallyLoaded=false
      drawAs=RoundedBox
      imageType=NoImage
      mirroring=NoMirror
      resourceName=None
      resourceObject=null
      tiling=NoTile
      imageSize=<0 0>
      margin=<ltrb 0 0 0 0>
      outlineSettings {
       bUseBrushTransparency=false
       roundingType=FixedRadius
       width=0
       color {colorUseRule=UseColor_Specified, specifiedColor=<rgba 0 0 0 0>}
       cornerRadii=<4 4 4 4>
      }
      tintColor {colorUseRule=UseColor_Specified, specifiedColor=<rgba 0.2159 0.2159 0.2159 1>}
      uVRegion {bIsValid=false, max=<0 0>, min=<0 0>}
     }
     horizontalBackgroundImage {
      bIsDynamicallyLoaded=false
      drawAs=NoDrawType
      imageType=NoImage
      mirroring=NoMirror
      resourceName=None
      resourceObject=null
      tiling=NoTile
      imageSize=<0 0>
      margin=<ltrb 0 0 0 0>
      outlineSettings {
       bUseBrushTransparency=false
       roundingType=HalfHeightRadius
       width=0
       color {colorUseRule=UseColor_Specified, specifiedColor=<rgba 0 0 0 0>}
       cornerRadii=<0 0 0 0>
      }
      tintColor {colorUseRule=UseColor_Specified, specifiedColor=<rgba 1 1 1 1>}
      uVRegion {bIsValid=false, max=<0 0>, min=<0 0>}
     }
     horizontalBottomSlotImage {
      bIsDynamicallyLoaded=false
      drawAs=NoDrawType
      imageType=NoImage
      mirroring=NoMirror
      resourceName=None
      resourceObject=null
      tiling=NoTile
      imageSize=<0 0>
      margin=<ltrb 0 0 0 0>
      outlineSettings {
       bUseBrushTransparency=false
       roundingType=HalfHeightRadius
       width=0
       color {colorUseRule=UseColor_Specified, specifiedColor=<rgba 0 0 0 0>}
       cornerRadii=<0 0 0 0>
      }
      tintColor {colorUseRule=UseColor_Specified, specifiedColor=<rgba 1 1 1 1>}
      uVRegion {bIsValid=false, max=<0 0>, min=<0 0>}
     }
     horizontalTopSlotImage {
      bIsDynamicallyLoaded=false
      drawAs=NoDrawType
      imageType=NoImage
      mirroring=NoMirror
      resourceName=None
      resourceObject=null
      tiling=NoTile
      imageSize=<0 0>
      margin=<ltrb 0 0 0 0>
      outlineSettings {
       bUseBrushTransparency=false
       roundingType=HalfHeightRadius
       width=0
       color {colorUseRule=UseColor_Specified, specifiedColor=<rgba 0 0 0 0>}
       cornerRadii=<0 0 0 0>
      }
      tintColor {colorUseRule=UseColor_Specified, specifiedColor=<rgba 1 1 1 1>}
      uVRegion {bIsValid=false, max=<0 0>, min=<0 0>}
     }
     hoveredThumbImage {
      bIsDynamicallyLoaded=false
      drawAs=RoundedBox
      imageType=NoImage
      mirroring=NoMirror
      resourceName=None
      resourceObject=null
      tiling=NoTile
      imageSize=<0 0>
      margin=<ltrb 0 0 0 0>
      outlineSettings {
       bUseBrushTransparency=false
       roundingType=FixedRadius
       width=0
       color {colorUseRule=UseColor_Specified, specifiedColor=<rgba 0 0 0 0>}
       cornerRadii=<4 4 4 4>
      }
      tintColor {colorUseRule=UseColor_Specified, specifiedColor=<rgba 0.2159 0.2159 0.2159 1>}
      uVRegion {bIsValid=false, max=<0 0>, min=<0 0>}
     }
     normalThumbImage {
      bIsDynamicallyLoaded=false
      drawAs=RoundedBox
      imageType=NoImage
      mirroring=NoMirror
      resourceName=None
      resourceObject=null
      tiling=NoTile
      imageSize=<0 0>
      margin=<ltrb 0 0 0 0>
      outlineSettings {
       bUseBrushTransparency=false
       roundingType=FixedRadius
       width=0
       color {colorUseRule=UseColor_Specified, specifiedColor=<rgba 0 0 0 0>}
       cornerRadii=<4 4 4 4>
      }
      tintColor {colorUseRule=UseColor_Specified, specifiedColor=<rgba 0.0953 0.0953 0.0953 1>}
      uVRegion {bIsValid=false, max=<0 0>, min=<0 0>}
     }
     verticalBackgroundImage {
      bIsDynamicallyLoaded=false
      drawAs=RoundedBox
      imageType=NoImage
      mirroring=NoMirror
      resourceName=None
      resourceObject=null
      tiling=NoTile
      imageSize=<0 0>
      margin=<ltrb 0 0 0 0>
      outlineSettings {
       bUseBrushTransparency=false
       roundingType=FixedRadius
       width=0
       color {colorUseRule=UseColor_Specified, specifiedColor=<rgba 0 0 0 0>}
       cornerRadii=<4 4 4 4>
      }
      tintColor {colorUseRule=UseColor_Specified, specifiedColor=<rgba 0.0083 0.0083 0.0083 1>}
      uVRegion {bIsValid=false, max=<0 0>, min=<0 0>}
     }
     verticalBottomSlotImage {
      bIsDynamicallyLoaded=false
      drawAs=NoDrawType
      imageType=NoImage
      mirroring=NoMirror
      resourceName=None
      resourceObject=null
      tiling=NoTile
      imageSize=<0 0>
      margin=<ltrb 0 0 0 0>
      outlineSettings {
       bUseBrushTransparency=false
       roundingType=HalfHeightRadius
       width=0
       color {colorUseRule=UseColor_Specified, specifiedColor=<rgba 0 0 0 0>}
       cornerRadii=<0 0 0 0>
      }
      tintColor {colorUseRule=UseColor_Specified, specifiedColor=<rgba 1 1 1 1>}
      uVRegion {bIsValid=false, max=<0 0>, min=<0 0>}
     }
     verticalTopSlotImage {
      bIsDynamicallyLoaded=false
      drawAs=NoDrawType
      imageType=NoImage
      mirroring=NoMirror
      resourceName=None
      resourceObject=null
      tiling=NoTile
      imageSize=<0 0>
      margin=<ltrb 0 0 0 0>
      outlineSettings {
       bUseBrushTransparency=false
       roundingType=HalfHeightRadius
       width=0
       color {colorUseRule=UseColor_Specified, specifiedColor=<rgba 0 0 0 0>}
       cornerRadii=<0 0 0 0>
      }
      tintColor {colorUseRule=UseColor_Specified, specifiedColor=<rgba 1 1 1 1>}
      uVRegion {bIsValid=false, max=<0 0>, min=<0 0>}
     }
    }}
   slot {class=VerticalBoxSlot, properties {padding=<ltrb 0 5 0 5>, size {sizeRule=Fill, value=1}}}
  }
  FxManagerDetails {
   widgetClassPath=&DetailsView
   parent=FxManagerDetailsScrollBox
   widget=FxManagerDetails
   properties {
    bAllowFiltering=true
    bAutoLoadAsset=false
    bShowAnimatedPropertiesOption=false
    bShowKeyablePropertiesOption=false
    bShowScrollBar=false
    propertiesToShow [FxManagerId, StageShowActors, ColorPalette, AutonomousLoop, AutonomousLoopEvents, StartLoopState, StopLoopState, IgnoreLoopConditions]
   }
   slot {class=ScrollBoxSlot, properties {padding=<ltrb 0 10 0 10>}}
  }
 ]
}
