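# An editor-only utility Blueprint in the Pilgrim quickplay StageManager tools that gives access to the Sparks FX Manager editor.
# Its single AddToFXManager function gets the Editor Utility Subsystem and spawns and registers the SparksFXManagerEditor widget as a tab.
# The function takes no inputs and passes nothing to the opened tool, and the event graph is empty.
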EditorUtilityBlueprint /PilgrimQuickplay/StageManager/NeverCook/EditorUtilities/EBP_FXManagerActionUtility
graphs [
 AddToFXManager {nodes [
   AddToFXManager() {is=title K2Node_FunctionEntry}
   > SpawnAndRegisterTab(self=Get EditorUtilitySubsystem(), InBlueprint=&/SparksFXManagerEditor)
  ]}
 EventGraph {links [], nodes []}
]
variables []
