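# Editor utility widget for editing the color array of a Mojo Color Shift palette asset in the Sparks stage show.
# Loading copies the palette's colors into an editable details view, or shows a warning if none is set.
# Auto Fill runs inline Python (CurveAtlasUtility) to sample the curve atlas at the slider's gradient point.
# Saving opens a transaction, writes the colors back and saves the asset; no EndTransaction follows.
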EditorUtilityWidgetBlueprint /PilgrimQuickplay/StageManager/NeverCook/EditorUtilities/SparksColorShiftDataEditor
graphs [
 EventGraph {nodes [
   "On Pressed (AutoFillButton)"() {is=title K2Node_ComponentBoundEvent}
   > Clear(TargetArray:Array of Linear Color Structures=Get Curves() -> (Curves:Array of Linear Color Structures))
   > n1 = Execute Python Script!(
    execute
    PythonScript="import sys\r\npath = project_content_dir+\"Python/TechArt/CurveAtlas/\"\r\nsys.path.append(project_content_dir+\"Python/TechArt/CurveAtlas/\")\r\nimport CurveAtlasUtility\r\nimport importlib\r\nimportlib.reload(CurveAtlasUtility)\r\ncurves = CurveAtlasUtility.GetColorsFromCurve(curve_atlas,curve_index)"
    project_content_dir:String=GetProjectContentDirectory!() -> (ReturnValue:String)
    curve_atlas:Curve Linear Color Atlas Object Reference=Get CurveAtlasReference() -> (CurveAtlasReference:Curve Linear Color Atlas Object Reference)
    curve_index:float=GetValue!(self:Slider Object Reference=Get Slider_ColorCurve() -> (Slider_ColorCurve:Editor Utility Slider Object Reference)) -> (ReturnValue:float)
   ) -> (then, ReturnValue, Curves:Array of Linear Color Structures) {inputs [project_content_dir, curve_atlas, curve_index], outputs [Curves]}
   > Set Curves(Curves:Array of Linear Color Structures=@n1.Curves) -> (Output_Get:Array of Linear Color Structures)
   n2 = "On Pressed (LoadDAButton)"() {is=title K2Node_ComponentBoundEvent}
   n3 = Get!(execute:Exec=[@n2.then, @n11.then]) -> (then:Exec, else:Exec, ColorPalette:BP Mojo Color Shift Palette Object Reference)
   > Set Curves(Curves=GetColors!(self:BP Mojo Color Shift Palette Object Reference=@n3) -> (Colors:Array of Linear Color Structures))
   > UpdateMessageView()
   UpdateMessageView(execute=@n3.else)
   "On Pressed (SaveChangesButton)"() {is=title K2Node_ComponentBoundEvent}
   > BeginTransaction(
    Context=Update ColorShift Data
    Description="LOCTEXT(\"ColorShift Data\")"
    PrimaryObject=Get ColorPalette() -> (ColorPalette:BP Mojo Color Shift Palette Object Reference)
   )
   > UpdateColors!(execute, self:BP Mojo Color Shift Palette Object Reference=Get ColorPalette(), Colors:Array of Linear Color Structures=Get Curves()) -> (then)
   n4 = Get ColorPalette()
   > TransactObject(Object=@n4)
   > SaveLoadedAsset(AssetToSave=@n4, bOnlyIfIsDirty=true)
   "On Value Changed (Slider_ColorCurve)"!() -> (OutputDelegate, then, Value:float) {is=title K2Node_ComponentBoundEvent}
   n5 = Get Slider_ColorCurve()
   n6 = Get MinValue!(self:Slider Object Reference=@n5) -> (MinValue:float)
   n7 = Get Value!(self:Slider Object Reference=@n5) -> (Value:float)
   n8 = SafeDivide(A=float - float(A=@n7, B=@n6), B=float - float(A=Get MaxValue!(self:Slider Object Reference=@n5) -> (MaxValue:float), B=@n6))
   n9 = Get Progress() -> (Progress:Simple Material Progress Bar Object Reference)
   > SetProgress!(execute, self:Fort Simple Material Progress Bar Object Reference=@n9, Progress:float=@n8) -> (then)
   n10 = "To Text (Float)"(Value=@n7, MinimumIntegralDigits=1, MaximumIntegralDigits=324, MinimumFractionalDigits=0, MaximumFractionalDigits=2)
   > "SetText (Text)"(self=Get Text_SettingValue() -> (Text_SettingValue:Text Object Reference), InText=@n10)
   n11 = LoadData() {is=title K2Node_CustomEvent}
   Event PreConstruct()
   > SetObject(
    self=[
     Get PaletteDAView() -> (PaletteDAView:Single Property View Object Reference)
     Get CurveAtlasReferenceViewer() -> (CurveAtlasReferenceViewer:Single Property View Object Reference)
     Get CurvesView() -> (CurvesView:Details View Object Reference)
    ]
    NewObject=Self-Reference()
   )
   Event Construct()
   > LoadData() {is=title K2Node_CallFunction}
   > SetProgress(self=@n9, Progress=@n8)
   > "SetText (Text)"(self=Get Text_SettingValue(), InText=@n10)
   Event Run()
   > RegisterTabAndGetID!(
    execute
    self:Editor Utility Subsystem Object Reference=Get EditorUtilitySubsystem()
    InBlueprint:Editor Utility Widget Blueprint Object Reference=&self
   ) -> (then, NewTabID:Name=None)
  ]}
 UpdateMessageView {nodes [
   UpdateMessageView() {is=title K2Node_FunctionEntry}
   > n1 = Is Valid(InputObject=Get ColorPalette())
   SetVisibility(
    execute=@n1.Is Not Valid
    self=[
     Get CurvesOverlay() -> (CurvesOverlay:Overlay Object Reference)
     Get CurveAtlasReferenceBorder() -> (CurveAtlasReferenceBorder:Border Object Reference)
    ]
    InVisibility=Collapsed
   )
   > SetVisibility(self=Get EmptyWarningOverlay() -> (EmptyWarningOverlay:Overlay Object Reference), InVisibility=Visible)
   SetVisibility(execute=@n1.Is Valid, self=[Get CurvesOverlay(), Get CurveAtlasReferenceBorder()], InVisibility=Visible)
   > SetVisibility(self=Get EmptyWarningOverlay(), InVisibility=Collapsed)
  ]}
]
variables [
 ColorPalette {type=BP_MojoColorShiftPalette_C, default=null}
 CurveAtlasReference {type=CurveLinearColorAtlas, default=/SparksStageShow/Systems/StageManager/FXManager/Curves/Mojo_StageShow_CurveAtlas}
 Curves {type="array<LinearColor>", default []}
]
widgets {
 info {rootWidgetClass=&CanvasPanel, widgetCount=48}
 widgets [
  CanvasPanel_0 {widgetClassPath=&CanvasPanel, widget=CanvasPanel_0}
  BG {
   widgetClassPath=&Image
   parent=CanvasPanel_0
   widget=BG
   properties {colorAndOpacity=<rgba 0.0176 0.0176 0.0176 1>}
   slot {class=CanvasPanelSlot, properties {layoutData {alignment=<0 0>, anchors {maximum=<1 1>, minimum=<0 0>}, offsets=<ltrb 0 0 0 0>}}}
  }
  ColorVerticalBox {
   widgetClassPath=&VerticalBox
   parent=CanvasPanel_0
   widget=ColorVerticalBox
   slot {class=CanvasPanelSlot, properties {layoutData {alignment=<0 0>, anchors {maximum=<1 1>, minimum=<0 0>}, offsets=<ltrb 0 0 0 0>}}}
  }
  Border_2 {
   widgetClassPath=&Border
   parent=ColorVerticalBox
   widget=Border_2
   properties {brushColor=<rgba 0.0176 0.0176 0.0176 1>, padding=<ltrb 10 20 10 20>}
   slot {class=VerticalBoxSlot}
  }
  HorizontalBox_1 {
   widgetClassPath=&HorizontalBox
   parent=Border_2
   widget=HorizontalBox_1
   slot {class=BorderSlot, properties {padding=<ltrb 10 20 10 20>}}
  }
  PaletteDAView {
   widgetClassPath=&SinglePropertyView
   parent=HorizontalBox_1
   widget=PaletteDAView
   properties {propertyName=ColorPalette}
   slot {class=HorizontalBoxSlot, properties {size {sizeRule=Fill, value=2}}}
  }
  SizeBox_2 {
   widgetClassPath=&SizeBox
   parent=HorizontalBox_1
   widget=SizeBox_2
   properties {bOverride_MaxDesiredHeight=true, maxDesiredHeight=28.3371}
   slot {class=HorizontalBoxSlot, properties {horizontalAlignment=HAlign_Right, verticalAlignment=VAlign_Top, size {sizeRule=Fill, value=1}}}
  }
  LoadDAButton {widgetClassPath=&EditorUtilityButton, parent=SizeBox_2, widget=LoadDAButton, slot {class=SizeBoxSlot}}
  ScaleBox {widgetClassPath=&ScaleBox, parent=LoadDAButton, widget=ScaleBox, slot {class=ButtonSlot}}
  TextBlock {widgetClassPath=&TextBlock, parent=ScaleBox, widget=TextBlock, properties {text=Load Data}, slot {class=ScaleBoxSlot}}
  Spacer_94 {widgetClassPath=&Spacer, parent=HorizontalBox_1, widget=Spacer_94, properties {size=<2.2305 1>}, slot {class=HorizontalBoxSlot}}
  SizeBox_3 {
   widgetClassPath=&SizeBox
   parent=HorizontalBox_1
   widget=SizeBox_3
   properties {bOverride_MaxDesiredHeight=true, maxDesiredHeight=28.3371}
   slot {class=HorizontalBoxSlot, properties {horizontalAlignment=HAlign_Right, verticalAlignment=VAlign_Top, size {sizeRule=Fill, value=1}}}
  }
  SaveChangesButton {widgetClassPath=&EditorUtilityButton, parent=SizeBox_3, widget=SaveChangesButton, slot {class=SizeBoxSlot}}
  ScaleBox_1 {widgetClassPath=&ScaleBox, parent=SaveChangesButton, widget=ScaleBox_1, slot {class=ButtonSlot}}
  TextBlock_1 {widgetClassPath=&TextBlock, parent=ScaleBox_1, widget=TextBlock_1, properties {text=Save Changes}, slot {class=ScaleBoxSlot}}
  SizeBox {
   widgetClassPath=&SizeBox
   parent=ColorVerticalBox
   widget=SizeBox
   properties {bOverride_HeightOverride=true, heightOverride=2}
   slot {class=VerticalBoxSlot, properties {verticalAlignment=VAlign_Bottom, padding=<ltrb 20 0 20 0>}}
  }
  Overlay {widgetClassPath=&Overlay, parent=SizeBox, widget=Overlay, slot {class=SizeBoxSlot}}
  LineShadow_01 {
   widgetClassPath=&Image
   parent=Overlay
   widget=LineShadow_01
   properties {colorAndOpacity=<rgba 0.0052 0.0052 0.0052 1>}
   slot {class=OverlaySlot, properties {horizontalAlignment=HAlign_Fill, verticalAlignment=VAlign_Fill, padding=<ltrb 0 1 0 -1>}}
  }
  Line_01 {
   widgetClassPath=&Image
   parent=Overlay
   widget=Line_01
   properties {colorAndOpacity=<rgba 0.0729 0.0729 0.0729 1>}
   slot {class=OverlaySlot, properties {horizontalAlignment=HAlign_Fill, verticalAlignment=VAlign_Fill}}
  }
  CurveAtlasReferenceBorder {
   widgetClassPath=&Border
   bIsVariable=true
   parent=ColorVerticalBox
   widget=CurveAtlasReferenceBorder
   properties {brushColor=<rgba 0.0176 0.0176 0.0176 1>, padding=<ltrb 10 20 10 20>}
   slot {class=VerticalBoxSlot}
  }
  Overlay_0 {
   widgetClassPath=&Overlay
   parent=CurveAtlasReferenceBorder
   widget=Overlay_0
   slot {class=BorderSlot, properties {padding=<ltrb 10 20 10 20>}}
  }
  SizeBox_1 {
   widgetClassPath=&SizeBox
   parent=Overlay_0
   widget=SizeBox_1
   properties {bOverride_HeightOverride=true, heightOverride=2}
   slot {class=OverlaySlot, properties {horizontalAlignment=HAlign_Fill, verticalAlignment=VAlign_Bottom, padding=<ltrb 5 0 5 -10>}}
  }
  Overlay_4 {widgetClassPath=&Overlay, parent=SizeBox_1, widget=Overlay_4, slot {class=SizeBoxSlot}}
  LineShadow_02 {
   widgetClassPath=&Image
   parent=Overlay_4
   widget=LineShadow_02
   properties {colorAndOpacity=<rgba 0.0052 0.0052 0.0052 1>}
   slot {class=OverlaySlot, properties {horizontalAlignment=HAlign_Fill, verticalAlignment=VAlign_Fill, padding=<ltrb 0 1 0 -1>}}
  }
  Line_02 {
   widgetClassPath=&Image
   parent=Overlay_4
   widget=Line_02
   properties {colorAndOpacity=<rgba 0.0729 0.0729 0.0729 1>}
   slot {class=OverlaySlot, properties {horizontalAlignment=HAlign_Fill, verticalAlignment=VAlign_Fill}}
  }
  HorizontalBox_0 {
   widgetClassPath=&HorizontalBox
   parent=Overlay_0
   widget=HorizontalBox_0
   slot {class=OverlaySlot, properties {horizontalAlignment=HAlign_Fill, verticalAlignment=VAlign_Fill}}
  }
  CurveAtlasReferenceViewer {
   widgetClassPath=&SinglePropertyView
   parent=HorizontalBox_0
   widget=CurveAtlasReferenceViewer
   properties {propertyName=CurveAtlasReference}
   slot {class=HorizontalBoxSlot, properties {size {sizeRule=Fill, value=1}}}
  }
  SizeBox_5 {
   widgetClassPath=&SizeBox
   parent=HorizontalBox_0
   widget=SizeBox_5
   properties {bOverride_MinDesiredWidth=true, minDesiredWidth=41.3939}
   slot {class=HorizontalBoxSlot, properties {verticalAlignment=VAlign_Top}}
  }
  Border_1 {
   widgetClassPath=&Border
   parent=SizeBox_5
   widget=Border_1
   properties {
    toolTipText="Choose which point of the curve of the color curve that you want to grab, it will apply on all curves."
    background {
     bIsDynamicallyLoaded=false
     drawAs=NoDrawType
     imageType=NoImage
     mirroring=NoMirror
     resourceName=None
     resourceObject=null
     tiling=NoTile
     imageSize=<32 32>
     margin=<ltrb 0 0 0 0>
     outlineSettings {
      bUseBrushTransparency=false
      roundingType=HalfHeightRadius
      width=0
      color {colorUseRule=UseColor_Specified, specifiedColor=<rgba 0 0 0 0>}
      cornerRadii=<0 0 0 0>
     }
     tintColor {colorUseRule=UseColor_Specified, specifiedColor=<rgba 1 1 1 1>}
     uVRegion {bIsValid=false, max=<0 0>, min=<0 0>}
    }
    padding=<ltrb -19 2 10 2>
   }
   slot {class=SizeBoxSlot}
  }
  Overlay_1 {widgetClassPath=&Overlay, parent=Border_1, widget=Overlay_1, slot {class=BorderSlot, properties {padding=<ltrb -19 2 10 2>}}}
  TextBlock_115 {
   widgetClassPath=&TextBlock
   parent=Overlay_1
   widget=TextBlock_115
   properties {
    text=Gradient Color Point:
    toolTipText="Choose which point of the curve of the color curve that you want to grab, it will apply on all curves."
    font {
     bForceMonospaced=false
     bMaterialIsStencil=false
     fontMaterial=null
     fontObject=&Roboto
     letterSpacing=0
     monospacedWidth=1
     size=9
     skewAmount=0
     typefaceFontName=Bold
     outlineSettings {
      bApplyOutlineToDropShadows=false
      bMiteredCorners=false
      bSeparateFillAlpha=false
      outlineMaterial=null
      outlineSize=0
      outlineColor=<rgba 0 0 0 1>
     }
    }
   }
   slot {class=OverlaySlot, properties {horizontalAlignment=HAlign_Center, padding=<ltrb 0 -20 0 0>}}
  }
  Border_0 {
   widgetClassPath=&Border
   parent=Overlay_1
   widget=Border_0
   properties {
    verticalAlignment=VAlign_Center
    background {
     bIsDynamicallyLoaded=false
     drawAs=NoDrawType
     imageType=NoImage
     mirroring=NoMirror
     resourceName=None
     resourceObject=null
     tiling=NoTile
     imageSize=<32 32>
     margin=<ltrb 0 0 0 0>
     outlineSettings {
      bUseBrushTransparency=false
      roundingType=HalfHeightRadius
      width=0
      color {colorUseRule=UseColor_Specified, specifiedColor=<rgba 0 0 0 0>}
      cornerRadii=<0 0 0 0>
     }
     tintColor {colorUseRule=UseColor_Specified, specifiedColor=<rgba 1 1 1 1>}
     uVRegion {bIsValid=false, max=<0 0>, min=<0 0>}
    }
    brushColor=<rgba 0.0521 0.0521 0.0521 0>
    padding=<ltrb 0 0 0 0>
   }
   slot {class=OverlaySlot, properties {horizontalAlignment=HAlign_Fill, verticalAlignment=VAlign_Fill}}
  }
  Overlay_2 {
   widgetClassPath=&Overlay
   parent=Border_0
   widget=Overlay_2
   slot {class=BorderSlot, properties {verticalAlignment=VAlign_Center, padding=<ltrb 0 0 0 0>}}
  }
  Progress {
   widgetClassPath=/Game/Athena/HUD/SimpleMaterialProgressBar
   bIsVariable=true
   parent=Overlay_2
   widget=Progress
   properties {
    bOverrideDefaultColorA=true
    bOverrideDefaultColorB=true
    bOverrideDefaultColorBackground=true
    designTime_Progress=0.792
    segmentEdge=15.4099
    cachedColorA=<rgba 0.1377 0.1515 0.1562 1>
    cachedColorB=<rgba 0.0737 0.0781 0.0708 1>
    cachedColorBackground=<rgba 0.0365 0.0365 0.0365 1>
   }
   slot {class=OverlaySlot, properties {horizontalAlignment=HAlign_Fill, verticalAlignment=VAlign_Fill}}
  }
  Text_SettingValue {
   widgetClassPath=&TextBlock
   bIsVariable=true
   parent=Overlay_2
   widget=Text_SettingValue
   properties {
    justification=Center
    text="{Value}"
    font {
     bForceMonospaced=false
     bMaterialIsStencil=false
     fontMaterial=null
     fontObject=&Roboto
     letterSpacing=0
     monospacedWidth=1
     size=7.5
     skewAmount=0
     typefaceFontName=Bold
     outlineSettings {
      bApplyOutlineToDropShadows=false
      bMiteredCorners=false
      bSeparateFillAlpha=false
      outlineMaterial=null
      outlineSize=0
      outlineColor=<rgba 0 0 0 1>
     }
    }
   }
   slot {class=OverlaySlot, properties {horizontalAlignment=HAlign_Fill, verticalAlignment=VAlign_Center}}
  }
  Slider_ColorCurve {
   widgetClassPath=/Script/Blutility.EditorUtilitySlider
   bIsVariable=true
   parent=Overlay_2
   widget=Slider_ColorCurve
   properties {
    indentHandle=true
    sliderBarColor=<rgba 1 1 1 0>
    sliderHandleColor=<rgba 1 1 1 0>
    widgetStyle {
     barThickness=4
     disabledBarImage {
      bIsDynamicallyLoaded=false
      drawAs=NoDrawType
      imageType=NoImage
      mirroring=NoMirror
      resourceName=None
      resourceObject=null
      tiling=NoTile
      imageSize=<0 0>
      margin=<ltrb 0 0 0 0>
      outlineSettings {
       bUseBrushTransparency=false
       roundingType=FixedRadius
       width=1
       color {colorUseRule=UseColor_ColorTable, specifiedColor=<rgba 0 0 0 0>}
       cornerRadii=<2 2 2 2>
      }
      tintColor {colorUseRule=UseColor_Specified, specifiedColor=<rgba 0.0048 0.0048 0.0048 1>}
      uVRegion {bIsValid=false, max=<0 0>, min=<0 0>}
     }
     disabledThumbImage {
      bIsDynamicallyLoaded=false
      drawAs=NoDrawType
      imageType=NoImage
      mirroring=NoMirror
      resourceName=None
      resourceObject=null
      tiling=NoTile
      imageSize=<8 8>
      margin=<ltrb 0 0 0 0>
      outlineSettings {
       bUseBrushTransparency=false
       roundingType=HalfHeightRadius
       width=0
       color {colorUseRule=UseColor_Specified, specifiedColor=<rgba 0 0 0 0>}
       cornerRadii=<0 0 0 0>
      }
      tintColor {colorUseRule=UseColor_Specified, specifiedColor=<rgba 0.5271 0.5271 0.5271 1>}
      uVRegion {bIsValid=false, max=<0 0>, min=<0 0>}
     }
     hoveredBarImage {
      bIsDynamicallyLoaded=false
      drawAs=NoDrawType
      imageType=NoImage
      mirroring=NoMirror
      resourceName=None
      resourceObject=null
      tiling=NoTile
      imageSize=<0 0>
      margin=<ltrb 0 0 0 0>
      outlineSettings {
       bUseBrushTransparency=false
       roundingType=FixedRadius
       width=1
       color {colorUseRule=UseColor_ColorTable, specifiedColor=<rgba 0 0 0 0>}
       cornerRadii=<2 2 2 2>
      }
      tintColor {colorUseRule=UseColor_Specified, specifiedColor=<rgba 0.0048 0.0048 0.0048 1>}
      uVRegion {bIsValid=false, max=<0 0>, min=<0 0>}
     }
     hoveredThumbImage {
      bIsDynamicallyLoaded=false
      drawAs=NoDrawType
      imageType=NoImage
      mirroring=NoMirror
      resourceName=None
      resourceObject=null
      tiling=NoTile
      imageSize=<8 8>
      margin=<ltrb 0 0 0 0>
      outlineSettings {
       bUseBrushTransparency=false
       roundingType=HalfHeightRadius
       width=0
       color {colorUseRule=UseColor_Specified, specifiedColor=<rgba 0 0 0 0>}
       cornerRadii=<0 0 0 0>
      }
      tintColor {colorUseRule=UseColor_Specified, specifiedColor=<rgba 1 1 1 1>}
      uVRegion {bIsValid=false, max=<0 0>, min=<0 0>}
     }
     normalBarImage {
      bIsDynamicallyLoaded=false
      drawAs=NoDrawType
      imageType=NoImage
      mirroring=NoMirror
      resourceName=None
      resourceObject=null
      tiling=NoTile
      imageSize=<0 0>
      margin=<ltrb 0 0 0 0>
      outlineSettings {
       bUseBrushTransparency=false
       roundingType=FixedRadius
       width=1
       color {colorUseRule=UseColor_ColorTable, specifiedColor=<rgba 0 0 0 0>}
       cornerRadii=<2 2 2 2>
      }
      tintColor {colorUseRule=UseColor_Specified, specifiedColor=<rgba 0.0048 0.0048 0.0048 1>}
      uVRegion {bIsValid=false, max=<0 0>, min=<0 0>}
     }
     normalThumbImage {
      bIsDynamicallyLoaded=false
      drawAs=NoDrawType
      imageType=NoImage
      mirroring=NoMirror
      resourceName=None
      resourceObject=null
      tiling=NoTile
      imageSize=<8 8>
      margin=<ltrb 0 0 0 0>
      outlineSettings {
       bUseBrushTransparency=false
       roundingType=HalfHeightRadius
       width=0
       color {colorUseRule=UseColor_Specified, specifiedColor=<rgba 0 0 0 0>}
       cornerRadii=<0 0 0 0>
      }
      tintColor {colorUseRule=UseColor_Specified, specifiedColor=<rgba 0.5271 0.5271 0.5271 1>}
      uVRegion {bIsValid=false, max=<0 0>, min=<0 0>}
     }
    }
   }
   slot {class=OverlaySlot, properties {horizontalAlignment=HAlign_Fill, verticalAlignment=VAlign_Fill}}
  }
  SizeBox_0 {
   widgetClassPath=&SizeBox
   parent=HorizontalBox_0
   widget=SizeBox_0
   properties {bOverride_MaxDesiredHeight=true, bOverride_MaxDesiredWidth=true, maxDesiredHeight=28.3371, maxDesiredWidth=80}
   slot {class=HorizontalBoxSlot, properties {horizontalAlignment=HAlign_Right, verticalAlignment=VAlign_Top}}
  }
  AutoFillButton {widgetClassPath=&EditorUtilityButton, parent=SizeBox_0, widget=AutoFillButton, slot {class=SizeBoxSlot}}
  ScaleBox_0 {widgetClassPath=&ScaleBox, parent=AutoFillButton, widget=ScaleBox_0, slot {class=ButtonSlot}}
  TextBlock_50 {widgetClassPath=&TextBlock, parent=ScaleBox_0, widget=TextBlock_50, properties {text=Auto Fill}, slot {class=ScaleBoxSlot}}
  CurvesOverlay {
   widgetClassPath=&Overlay
   bIsVariable=true
   parent=ColorVerticalBox
   widget=CurvesOverlay
   properties {clipping=ClipToBounds}
   slot {class=VerticalBoxSlot, properties {verticalAlignment=VAlign_Top, size {sizeRule=Fill, value=1}}}
  }
  CurvesView {
   widgetClassPath=&DetailsView
   parent=CurvesOverlay
   widget=CurvesView
   properties {propertiesToShow [Curves]}
   slot {class=OverlaySlot, properties {horizontalAlignment=HAlign_Fill, verticalAlignment=VAlign_Fill, padding=<ltrb 0 -30 0 0>}}
  }
  EmptyWarningOverlay {widgetClassPath=&Overlay, bIsVariable=true, parent=ColorVerticalBox, widget=EmptyWarningOverlay, slot {class=VerticalBoxSlot}}
  HorizontalBox_2 {
   widgetClassPath=&HorizontalBox
   parent=EmptyWarningOverlay
   widget=HorizontalBox_2
   slot {class=OverlaySlot, properties {horizontalAlignment=HAlign_Center, padding=<ltrb 0 20 0 0>}}
  }
  SizeBox_4 {
   widgetClassPath=&SizeBox
   parent=HorizontalBox_2
   widget=SizeBox_4
   properties {bOverride_MaxDesiredHeight=true, bOverride_MaxDesiredWidth=true, maxDesiredHeight=20, maxDesiredWidth=20}
   slot {class=HorizontalBoxSlot, properties {verticalAlignment=VAlign_Center}}
  }
  Exclamation {
   widgetClassPath=&Image
   parent=SizeBox_4
   widget=Exclamation
   properties {
    brush {
     bIsDynamicallyLoaded=false
     drawAs=Image
     imageType=NoImage
     mirroring=NoMirror
     resourceName=None
     resourceObject=/Game/UI/Foundation/Textures/Icons/Replay/T-Icon-Replay-Exclamation-L
     tiling=NoTile
     imageSize=<32 32>
     margin=<ltrb 0 0 0 0>
     outlineSettings {
      bUseBrushTransparency=false
      roundingType=HalfHeightRadius
      width=0
      color {colorUseRule=UseColor_Specified, specifiedColor=<rgba 0 0 0 0>}
      cornerRadii=<0 0 0 0>
     }
     tintColor {colorUseRule=UseColor_Specified, specifiedColor=<rgba 1 1 1 1>}
     uVRegion {bIsValid=false, max=<0 0>, min=<0 0>}
    }
    colorAndOpacity=<rgba 1 0.8114 0.105 1>
   }
   slot {class=SizeBoxSlot}
  }
  Spacer_99 {
   widgetClassPath=&Spacer
   parent=HorizontalBox_2
   widget=Spacer_99
   properties {size=<10 1>}
   slot {class=HorizontalBoxSlot, properties {horizontalAlignment=HAlign_Left, verticalAlignment=VAlign_Top}}
  }
  TextBlock_64 {
   widgetClassPath=&TextBlock
   parent=HorizontalBox_2
   widget=TextBlock_64
   properties {
    text=Please Load a Color Palette
    font {
     bForceMonospaced=false
     bMaterialIsStencil=false
     fontMaterial=null
     fontObject=&Roboto
     letterSpacing=0
     monospacedWidth=1
     size=11.25
     skewAmount=0
     typefaceFontName=Bold
     outlineSettings {
      bApplyOutlineToDropShadows=false
      bMiteredCorners=false
      bSeparateFillAlpha=false
      outlineMaterial=null
      outlineSize=0
      outlineColor=<rgba 0 0 0 1>
     }
    }
   }
   slot {class=HorizontalBoxSlot}
  }
 ]
}
